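; Dev/debug-only shotgun weapon Blueprint that fakes spread impact FX on fire, since the ability does the real targeting.
; It reads long range from weapon stats and builds a five-point X from the aim direction and pitch/yaw spread angles.
; Line traces to those points play native impacts on non-pawn hits, leaving pawns to ability FX.
; A flag switches to spawning tracers from stored points; debug draws the pattern.
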(event EventOnPlayWeaponFireFX
  (params (bPersistentFire :type Boolean) (bSecondaryFire :type Boolean))
  (bind (_aimdir _returnvalue) (FortWeapon|GetDamageStartLocation -1))
  (bind (_aimdir_2 _returnvalue_2) (FortWeapon|GetDamageStartLocation -1))
  (bind (_aimdir_3 _returnvalue_3) (FortWeapon|GetDamageStartLocation -1))
  (bind _currentfxtracewashit (Variables|InternalVariables|GetCurrentFXTraceWasHit))
  (bind _returnvalue_4 (Math|Color|SelectColor "(R=1.000000,G=0.000000,B=0.000000,A=1.000000)" "(R=0.000000,G=0.000000,B=1.000000,A=1.000000)" _currentfxtracewashit))
  (bind (_aimdir_4 _returnvalue_5) (FortWeapon|GetDamageStartLocation -1))
  (bind (_aimdir_5 _returnvalue_6) (FortWeapon|GetDamageStartLocation -1))
  (bind _aimdir_6 (FortWeapon|GetDamageStartLocation -1))
  (bind _returnvalue_7 (Math|Conversions|RotationFromXVector _aimdir_6))
  (bind _fxtracepitchfullangle (Variables|Default|GetFXTracePitchFullAngle))
  (bind _returnvalue_8 (/ _fxtracepitchfullangle 2))
  (bind _returnvalue_9 (Math|Trig|Tan(Degrees) _returnvalue_8))
  (bind _fxtraceyawfullangle (Variables|Default|GetFXTraceYawFullAngle))
  (bind _returnvalue_10 (/ _fxtraceyawfullangle 2))
  (bind _returnvalue_11 (Math|Trig|Tan(Degrees) _returnvalue_10))
  (bind (_x _y _z) (Math|Rotator|BreakRotIntoAxes _returnvalue_7))
  (bind _aimdir_7 (FortWeapon|GetDamageStartLocation -1))
  (bind _returnvalue_12 (Math|Vector|RotateVectorAroundAxis _aimdir_7 90 _z))
  (bind _aimdir_8 (FortWeapon|GetDamageStartLocation -1))
  (bind _returnvalue_13 (Math|Conversions|RotationFromXVector _aimdir_8))
  (bind (_x_2 _y_2) (Math|Rotator|BreakRotIntoAxes _returnvalue_13))
  (bind _aimdir_9 (FortWeapon|GetDamageStartLocation -1))
  (bind _returnvalue_14 (Math|Vector|RotateVectorAroundAxis _aimdir_9 90 _y_2))
  (bind _mysteriousrangebufferfortracers (Variables|InternalVariables|GetMysteriousRangeBufferForTracers))
  (bind _rangelong (Variables|Default|GetRangeLong))
  (bind _returnvalue_15 (- _rangelong _mysteriousrangebufferfortracers))
  (bind _returnvalue_16 (* _returnvalue_11 _returnvalue_15))
  (bind _returnvalue_17 (Math|Vector|RotateVectorAroundAxis _aimdir_7 -90 _z))
  (bind _returnvalue_18 (* _returnvalue_17 _returnvalue_16))
  (bind _returnvalue_19 (Math|Vector|RotateVectorAroundAxis _aimdir_9 -90 _y_2))
  (bind _returnvalue_20 (* _returnvalue_12 _returnvalue_16))
  (bind _mysteriousrangebufferfortracers_2 (Variables|InternalVariables|GetMysteriousRangeBufferForTracers))
  (bind _rangelong_2 (Variables|Default|GetRangeLong))
  (bind _returnvalue_21 (- _rangelong_2 _mysteriousrangebufferfortracers_2))
  (bind _returnvalue_22 (* _returnvalue_9 _returnvalue_21))
  (bind _returnvalue_23 (* _returnvalue_14 _returnvalue_22))
  (bind _returnvalue_24 (* _returnvalue_19 _returnvalue_22))
  (bind _endofcenterfxtrace (Variables|InternalVariables|GetEndOfCenterFXTrace))
  (bind _returnvalue_25 (+ _endofcenterfxtrace _returnvalue_18))
  (bind _endofcenterfxtrace_2 (Variables|InternalVariables|GetEndOfCenterFXTrace))
  (bind _returnvalue_26 (+ _endofcenterfxtrace_2 _returnvalue_20))
  (bind _returnvalue_27 (+ _returnvalue_25 _returnvalue_24))
  (bind _returnvalue_28 (+ _returnvalue_26 _returnvalue_23))
  (bind _returnvalue_29 (+ _returnvalue_25 _returnvalue_23))
  (bind _returnvalue_30 (+ _returnvalue_26 _returnvalue_24))
  (bind _currentfxtracewashit_2 (Variables|InternalVariables|GetCurrentFXTraceWasHit))
  (bind _currentfxtraceend (Variables|InternalVariables|GetCurrentFXTraceEnd))
  ;[1 1 1] Add fake impact effects, because the ability is doing the real targeting.
  ;
  ; Creates 5 points in an X shape to trace out the size of the box so you can tell what you might be able to hit.
  (|CallParentFunction bPersistentFire false)
  (bind _asfort_weapon_ranged_item_definition (Utilities|Casting|CastToFortWeaponRangedItemDefinition (Data|GetWeaponData)))
  (bind _outrow (Fortnite|GetRangedWeaponStatsRow (WeaponItemDefinition|GetWeaponStatHandle _asfort_weapon_ranged_item_definition)))
  (Variables|Default|SetRangeLong (.RngLong (Utilities|Struct|BreakFortRangedWeaponStats _outrow)))
  (Development|PrintString (Utilities|String|BuildString(Float) (Utilities|String|BuildString(Float) (Utilities|String|BuildString(Float) (Utilities|String|BuildString(Float) "" "PB=" (.RngPB (Utilities|Struct|BreakFortRangedWeaponStats _outrow)) " | ") "MID=" (.RngMid (Utilities|Struct|BreakFortRangedWeaponStats _outrow)) " | ") "LONG=" (.RngLong (Utilities|Struct|BreakFortRangedWeaponStats _outrow)) " | ") "MAX=" (.RngMax (Utilities|Struct|BreakFortRangedWeaponStats _outrow))) (Variables|Default|GetDebug) (Variables|Default|GetDebug))
  (Utilities|Array|Clear (Variables|InternalVariables|GetAllFXTracerSpawnPoints))
  (Variables|InternalVariables|SetEndOfCenterFXTrace (+ _returnvalue_3 (* _aimdir_3 (- (Variables|Default|GetRangeLong) (Variables|InternalVariables|GetMysteriousRangeBufferForTracers)))))
  ;; Set Trace Points
  (Variables|InternalVariables|SetAllFXTraceEndPoints (Utilities|Array|MakeArray (Variables|InternalVariables|GetEndOfCenterFXTrace) _returnvalue_30 _returnvalue_28 _returnvalue_29 _returnvalue_27))
  (if (Variables|Default|GetDebug)
    (Rendering|Debug|DrawDebugLine _endofcenterfxtrace _returnvalue_25 "(R=1.000000,G=1.000000,B=0.000000,A=1.000000)" 5) ; Left
    (Rendering|Debug|DrawDebugLine _endofcenterfxtrace_2 _returnvalue_26 "(R=1.000000,G=1.000000,B=0.000000,A=1.000000)" 5) ; Right
    (Rendering|Debug|DrawDebugLine (Variables|InternalVariables|GetEndOfCenterFXTrace) (+ (Variables|InternalVariables|GetEndOfCenterFXTrace) _returnvalue_23) "(R=0.000000,G=1.000000,B=0.000000,A=1.000000)" 5) ; Up
    (Rendering|Debug|DrawDebugLine (Variables|InternalVariables|GetEndOfCenterFXTrace) (+ (Variables|InternalVariables|GetEndOfCenterFXTrace) _returnvalue_24) "(R=0.000000,G=1.000000,B=0.000000,A=1.000000)" 5) ; Down
    (Rendering|Debug|DrawDebugLine _returnvalue_25 _returnvalue_27 "(R=1.000000,G=0.000000,B=0.000000,A=1.000000)" 5) ; Left/Up
    (Rendering|Debug|DrawDebugLine _returnvalue_26 _returnvalue_30 "(R=1.000000,G=0.016000,B=0.000000,A=1.000000)" 5) ; Right/Up
    (Rendering|Debug|DrawDebugLine _returnvalue_25 _returnvalue_29 "(R=1.000000,G=0.000000,B=0.000000,A=1.000000)" 5) ; Left/Down
    (Rendering|Debug|DrawDebugLine _returnvalue_26 _returnvalue_28 "(R=1.000000,G=0.005000,B=0.000000,A=1.000000)" 5) ; Right/Down
  )
  (for (_array_element _array_index) (Variables|InternalVariables|GetAllFXTraceEndPoints)
    (Variables|InternalVariables|SetCurrentFXTraceEnd _array_element)
    (bind _outhit (Collision|LineTraceByChannel _returnvalue_6 (Variables|InternalVariables|GetCurrentFXTraceEnd) "TraceTypeQuery8" true (Utilities|Array|MakeArray (Game|GetInstigator self))))
    (bind _output_get (Variables|InternalVariables|SetCurrentFXTraceWasHit (Collision|BreakHitResult _outhit)))
    (if (Variables|Default|GetSpawnTracersInsteadOfDummyImpactFX)
      _
      (elif _output_get
        (Utilities|Casting|CastToFortPawn (.HitActor (Collision|BreakHitResult _outhit))
          (:CastFailed
            (FortWeapon|PlayNativeImpactFX (select (Variables|InternalVariables|GetCurrentFXTraceWasHit) _outhit (Collision|MakeHitResult false false 1 0.0 (Variables|InternalVariables|GetCurrentFXTraceEnd) (Variables|InternalVariables|GetCurrentFXTraceEnd) "0, 0, 0" "0, 0, 0" 0 self 0 "None" "None" 0 0 0 _returnvalue_2 (Variables|InternalVariables|GetCurrentFXTraceEnd))) (Ability|GameplayCue|MakeGameplayCueParameters :NormalizedMagnitude 0.0 :RawMagnitude 0.0 :EffectContext "()" :AggregatedSourceTags "(GameplayTags=)" :AggregatedTargetTags "(GameplayTags=)" :Location "0,0,0" :Normal "0,0,0" :GameplayEffectLevel 1 :AbilityLevel 1 :bReplicateLocationWhenUsingMinimalRepProxy false))
            (Utilities|Array|Add (Variables|InternalVariables|GetAllFXTracerSpawnPoints) (Math|Vector|SelectVector (.Location (Collision|BreakHitResult _outhit)) _currentfxtraceend _currentfxtracewashit_2)) ; If this trace hit something, store that value so we can spawn tracers there.
            (if (Variables|Default|GetDebug)
              (Development|PrintString (Utilities|String|BuildString(Vector) (Utilities|String|BuildString(Integer) "Trace Loc  " "[" _array_index "]") " = " (Math|Vector|SelectVector (.Location (Collision|BreakHitResult _outhit)) _currentfxtraceend _currentfxtracewashit_2)) (Variables|Default|GetDebug) (Variables|Default|GetDebug))
              (Rendering|Debug|DrawDebugSphere (Math|Vector|SelectVector (.Location (Collision|BreakHitResult _outhit)) _currentfxtraceend _currentfxtracewashit_2) 10 12 _returnvalue_4 5)
              (Rendering|Debug|DrawDebugLine _returnvalue (Variables|InternalVariables|GetCurrentFXTraceEnd) _returnvalue_4 5)))) ; Skip tracing if this hits a pawn. Let the ability FX handle that.
      )))
  ;[1 1 1] Spawn tracers in X pattern instead of just impact FX
  (if (Variables|Default|GetSpawnTracersInsteadOfDummyImpactFX)
    (for _array_element_2 (Variables|InternalVariables|GetAllFXTracerSpawnPoints)
      (if (Variables|Default|GetDebug)
        (Rendering|Debug|DrawDebugSphere _array_element_2 20 12 "(R=1.000000,G=1.000000,B=1.000000,A=1.000000)" 5)
        (else
          _))
      (bind _returnvalue_31 (Game|SpawnActorfromClass (Effects|GetTracerTemplate) (Math|Transform|MakeTransform _array_element_2 (Math|Conversions|RotationFromXVector (Math|Vector|Normalize (- _array_element_2 _returnvalue_5)))) "AlwaysSpawn" "SelectDefaultAtRuntime" self (Game|GetInstigator self)))
      (Transformation|SetActorRotation _returnvalue_31 (Math|Conversions|RotationFromXVector (Math|Vector|Normalize (- _array_element_2 _returnvalue_5))) false) ; Seem to need to do this after spawning.
    )
    (Development|PrintString "===== DONE =====" (Variables|Default|GetDebug) (Variables|Default|GetDebug))))

(fn ConstructionScript
  (|CallParentFunction))
